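; Gameplay ability for a quest that tracks repeated triggering events, named for a "2 in 30" condition.
; On each event activation, it calls the parent, then applies GE_Quest_2in30 to the owner as a counter.
; If the owning pawn's count of that effect is at least two, it sends a custom stat event with source tags to the player's quest manager.
; The ability then ends; any time window is left to the effect.
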(event Ability|EventActivateAbilityFromEvent
  (params (EventData :type Gameplay%20Event%20Data%20Structure%20%28by%20ref%29))
  (|CallParentFunction EventData)
  (Ability|ApplyGameplayEffectToOwner (Variables|Default|GetGE_Quest_2in30))
  (if (>= (GameplayEffects|GetGameplayEffectCount (Ability|GetAbilitySystemComponent (Variables|GAVars|Data|GetOwningPlayerPawn)) "/Close/Items/QuestAbilities/GE_Quest_2in30.GE_Quest_2in30_C" 0 false) 2)
    (Quest|SendComplexCustomStatEvent (Variables|GAVars|Data|GetOwningPlayerQuestManager) 0 (Variables|Default|GetT%20Source%20Tags) "(GameplayTags=)" 1))
  (Ability|EndAbility))
